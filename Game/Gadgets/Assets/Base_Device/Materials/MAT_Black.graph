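# MAT_Black is a Default Lit material in the gadget Base_Device assets, presumably a plain black surface, going by its name.
# Its graph holds a single vector parameter at default settings, with no connection to a material output shown.
# It is flagged for use with skeletal meshes, Nanite and mesh deformers.
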Material /Game/Gadgets/Assets/Base_Device/Materials/MAT_Black
graph {
 properties {
  bUsedWithMeshDeformer=true
  bUsedWithNanite=true
  bUsedWithSkeletalMesh=true
  pixelDepthOffsetMode=PDOM_Legacy
  usedShadingModels=Default Lit
 }
 nodes [VectorParameter()]
}
